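# Physics asset for the left-arm poncho dynamic bones on the male medium Headband adult character.
# A chain of elbow poncho bones, 03 through 07, hangs from a tiny master body on the upper arm, with limited swing and tight twist between links.
# Capsules cover the pelvis, spine, left arm and left leg, joined by free constraints.
# Several poncho links carry a second capsule, offset or angled to widen their coverage.
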PhysicsAsset /BRCosmetics/Characters/Player/Male/Medium/Bodies/M_MED_Headband_Adult/Meshes/M_MED_Headband_Adult_Dyn_poncho_arm_l
bodies {
 calf_l [{shapeType=Capsule, length=28.625, radius=10.1638, center=<-21.0255 -1.8396 0>, rotation=<rot 90 6.5366 3.1936>}]
 dyn_poncho_elbow_l_03 [
  {shapeType=Capsule, length=26.4162, radius=9, shapeName=dyn_poncho_elbow_l_03_capsule, center=<10.3553 0 7.8312>, rotation=<rot 81.0072 0 0>}
  {
   shapeType=Capsule
   length=26.4162
   radius=9
   shapeName=dyn_poncho_elbow_l_03_capsule
   center=<42.7943 0.0002 3.9403>
   rotation=<rot 81.0072 0 0>
  }
 ]
 dyn_poncho_elbow_l_04 [
  {shapeType=Capsule, length=26.4162, radius=9, shapeName=dyn_poncho_elbow_l_03_capsule, center=<10.5403 0 9>, rotation=<rot 90 0 0>}
  {
   shapeType=Capsule
   length=26.4162
   radius=9
   shapeName=dyn_poncho_elbow_l_07_capsule
   center=<10.4741 -13.0916 4.1459>
   rotation=<rot 68.6316 -89.9998 -89.9997>
  }
 ]
 dyn_poncho_elbow_l_05 [{shapeType=Capsule, length=26.4162, radius=9, shapeName=dyn_poncho_elbow_l_03_capsule, center=<10.5403 0 9>, rotation=<rot 90 0 0>}]
 dyn_poncho_elbow_l_06 [
  {shapeType=Capsule, length=26.4162, radius=9, shapeName=dyn_poncho_elbow_l_03_capsule, center=<10.5403 0 9>, rotation=<rot 90 0 0>}
  {
   shapeType=Capsule
   length=26.4162
   radius=9
   shapeName=dyn_poncho_elbow_l_07_capsule
   center=<10.4741 -13.0916 4.1459>
   rotation=<rot 68.6316 -89.9998 -89.9997>
  }
 ]
 dyn_poncho_elbow_l_07 [
  {shapeType=Capsule, length=26.4162, radius=9, shapeName=dyn_poncho_elbow_l_03_capsule, center=<10.5403 0 9>, rotation=<rot 90 0 0>}
  {
   shapeType=Capsule
   length=26.4162
   radius=9
   shapeName=dyn_poncho_elbow_l_07_capsule
   center=<10.4741 -13.0916 4.1459>
   rotation=<rot 68.6316 -89.9998 -89.9997>
  }
 ]
 dyn_poncho_elbow_l_MSTR [{shapeType=Capsule, length=0.0683, radius=0.5, shapeName=dyn_poncho_elbow_l_03_capsule, rotation=<rot 90 0 0>}]
 lowerarm_l [{shapeType=Capsule, length=17.9062, radius=6.695, center=<9.5574 -0.0005 0>, rotation=<rot 90 0 0>}]
 pelvis [
  {shapeType=Capsule, length=10.7187, radius=14.7263, center=<-2.6215 -0.3628 0>}
  {shapeType=Capsule, length=4.5937, radius=17.82, shapeName=pelvis_capsule, center=<-11.5935 0.2438 0>}
 ]
 spine_03 [{shapeType=Capsule, length=2.375, radius=18.3825, center=<-2.1747 2.4886 0>}]
 spine_05 [
  {shapeType=Capsule, length=13.1562, radius=12.2888, center=<-2.1747 -1.0553 0>}
  {shapeType=Capsule, length=17.6562, radius=10.8125, center=<-0.542 -1.0479 0>}
 ]
 thigh_l [{shapeType=Capsule, length=33, radius=9.9138, center=<-22.0309 -0.0003 0>, rotation=<rot 90 0 0>}]
 upperarm_l [{shapeType=Capsule, length=22, radius=8.1325, center=<11.436 -0.6303 -0.5399>, rotation=<rot 88.1564 -173.3738 179.9998>}]
}
constraints [
 {bone1Name=dyn_poncho_elbow_l_03, bone2Name=dyn_poncho_elbow_l_MSTR, swing1LimitDegrees=20, swing2LimitDegrees=45, twistLimitDegrees=5}
 {
  bone1Name=upperarm_l
  bone2Name=spine_05
  swing1LimitDegrees=45
  swing1Motion=Free
  swing2LimitDegrees=45
  swing2Motion=Free
  twistLimitDegrees=45
  twistMotion=Free
 }
 {
  bone1Name=lowerarm_l
  bone2Name=upperarm_l
  swing1LimitDegrees=45
  swing1Motion=Free
  swing2LimitDegrees=45
  swing2Motion=Free
  twistLimitDegrees=45
  twistMotion=Free
 }
 {
  bone1Name=thigh_l
  bone2Name=pelvis
  swing1LimitDegrees=45
  swing1Motion=Free
  swing2LimitDegrees=45
  swing2Motion=Free
  twistLimitDegrees=45
  twistMotion=Free
 }
 {
  bone1Name=calf_l
  bone2Name=thigh_l
  swing1LimitDegrees=45
  swing1Motion=Free
  swing2LimitDegrees=45
  swing2Motion=Free
  twistLimitDegrees=45
  twistMotion=Free
 }
 {bone1Name=dyn_poncho_elbow_l_04, bone2Name=dyn_poncho_elbow_l_03, swing1LimitDegrees=10, swing2LimitDegrees=45, twistLimitDegrees=5}
 {bone1Name=dyn_poncho_elbow_l_05, bone2Name=dyn_poncho_elbow_l_04, swing1LimitDegrees=10, swing2LimitDegrees=45, twistLimitDegrees=5}
 {bone1Name=dyn_poncho_elbow_l_06, bone2Name=dyn_poncho_elbow_l_05, swing1LimitDegrees=10, swing2LimitDegrees=45, twistLimitDegrees=5}
 {bone1Name=dyn_poncho_elbow_l_07, bone2Name=dyn_poncho_elbow_l_06, swing1LimitDegrees=10, swing2LimitDegrees=45, twistLimitDegrees=5}
 {
  bone1Name=dyn_poncho_elbow_l_MSTR
  bone2Name=upperarm_l
  swing1LimitDegrees=45
  swing1Motion=Free
  swing2LimitDegrees=45
  swing2Motion=Free
  twistLimitDegrees=45
  twistMotion=Free
 }
 {
  bone1Name=spine_05
  bone2Name=spine_03
  swing1LimitDegrees=45
  swing1Motion=Free
  swing2LimitDegrees=45
  swing2Motion=Free
  twistLimitDegrees=45
  twistMotion=Free
 }
 {
  bone1Name=spine_03
  bone2Name=pelvis
  swing1LimitDegrees=45
  swing1Motion=Free
  swing2LimitDegrees=45
  swing2Motion=Free
  twistLimitDegrees=45
  twistMotion=Free
 }
]
