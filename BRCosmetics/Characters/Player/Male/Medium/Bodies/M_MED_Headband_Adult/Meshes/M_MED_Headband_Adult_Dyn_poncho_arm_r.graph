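# Physics asset for the right-arm poncho dynamics on the medium male Headband adult player body.
# A chain of capsules, dyn_poncho_elbow_r_03 to 07, hangs from a tiny MSTR root on upperarm_r, with tight swing and twist limits between links.
# Capsules on spine, pelvis, right arm and right leg give body volumes, oversized and offset at pelvis and leg.
# Their constraints, like the root's, swing and twist freely.
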PhysicsAsset /BRCosmetics/Characters/Player/Male/Medium/Bodies/M_MED_Headband_Adult/Meshes/M_MED_Headband_Adult_Dyn_poncho_arm_r
bodies {
 calf_r [{shapeType=Capsule, length=11.375, radius=29.2575, center=<19.5286 -9.7827 -17.5368>, rotation=<rot 90 0 -173.5005>}]
 dyn_poncho_elbow_r_03 [{shapeType=Capsule, length=26.4162, radius=8.5, shapeName=dyn_poncho_elbow_l_03_capsule, center=<10.5403 0 10>, rotation=<rot 90 0 0>}]
 dyn_poncho_elbow_r_04 [{shapeType=Capsule, length=26.4162, radius=8.5, shapeName=dyn_poncho_elbow_l_03_capsule, center=<10.5403 0 9>, rotation=<rot 90 0 0>}]
 dyn_poncho_elbow_r_05 [{shapeType=Capsule, length=26.4162, radius=9, shapeName=dyn_poncho_elbow_l_03_capsule, center=<10.5403 0 9>, rotation=<rot 90 0 0>}]
 dyn_poncho_elbow_r_06 [{shapeType=Capsule, length=26.4162, radius=9, shapeName=dyn_poncho_elbow_l_03_capsule, center=<10.5403 0 9>, rotation=<rot 90 0 0>}]
 dyn_poncho_elbow_r_07 [{shapeType=Capsule, length=26.4162, radius=9, shapeName=dyn_poncho_elbow_l_03_capsule, center=<10.5403 0 9>, rotation=<rot 90 0 0>}]
 dyn_poncho_elbow_r_MSTR [{shapeType=Capsule, length=0.0683, radius=0.5, shapeName=dyn_poncho_elbow_l_03_capsule, rotation=<rot 90 0 0>}]
 lowerarm_r [{shapeType=Capsule, length=20, radius=6.7, center=<-9.5574 0.0005 0>, rotation=<rot 90 0 180>}]
 pelvis [{shapeType=Capsule, length=50.8437, radius=64.1951, shapeName=pelvis_capsule, center=<-34.4248 11.4113 -36.6589>, rotation=<rot 90 0 0>}]
 spine_03 [{shapeType=Capsule, length=2.3749, radius=18.3825, center=<-2.1747 2.4886 0>}]
 spine_05 [{shapeType=Capsule, length=13.1562, radius=12.2888, center=<-2.1747 -1.0553 0>}]
 thigh_r [{shapeType=Capsule, length=15, radius=29.9763, center=<22.0309 -12.706 -17.5368>, rotation=<rot 90 0 -179.9999>}]
 upperarm_r [{shapeType=Capsule, length=22.0937, radius=8.0388, center=<-11.4839 0.9823 1.4365>, rotation=<rot 86.8153 -173.0089 2.8009>}]
}
constraints [
 {bone1Name=dyn_poncho_elbow_r_03, bone2Name=dyn_poncho_elbow_r_MSTR, swing1LimitDegrees=20, swing2LimitDegrees=45, twistLimitDegrees=5}
 {
  bone1Name=upperarm_r
  bone2Name=spine_05
  swing1LimitDegrees=45
  swing1Motion=Free
  swing2LimitDegrees=45
  swing2Motion=Free
  twistLimitDegrees=45
  twistMotion=Free
 }
 {
  bone1Name=lowerarm_r
  bone2Name=upperarm_r
  swing1LimitDegrees=45
  swing1Motion=Free
  swing2LimitDegrees=45
  swing2Motion=Free
  twistLimitDegrees=45
  twistMotion=Free
 }
 {
  bone1Name=thigh_r
  bone2Name=pelvis
  swing1LimitDegrees=45
  swing1Motion=Free
  swing2LimitDegrees=45
  swing2Motion=Free
  twistLimitDegrees=45
  twistMotion=Free
 }
 {
  bone1Name=calf_r
  bone2Name=thigh_r
  swing1LimitDegrees=45
  swing1Motion=Free
  swing2LimitDegrees=45
  swing2Motion=Free
  twistLimitDegrees=45
  twistMotion=Free
 }
 {bone1Name=dyn_poncho_elbow_r_04, bone2Name=dyn_poncho_elbow_r_03, swing1LimitDegrees=20, swing2LimitDegrees=45, twistLimitDegrees=5}
 {bone1Name=dyn_poncho_elbow_r_05, bone2Name=dyn_poncho_elbow_r_04, swing1LimitDegrees=20, swing2LimitDegrees=45, twistLimitDegrees=5}
 {bone1Name=dyn_poncho_elbow_r_06, bone2Name=dyn_poncho_elbow_r_05, swing1LimitDegrees=20, swing2LimitDegrees=45, twistLimitDegrees=5}
 {bone1Name=dyn_poncho_elbow_r_07, bone2Name=dyn_poncho_elbow_r_06, swing1LimitDegrees=20, swing2LimitDegrees=45, twistLimitDegrees=5}
 {
  bone1Name=dyn_poncho_elbow_r_MSTR
  bone2Name=upperarm_r
  swing1LimitDegrees=45
  swing1Motion=Free
  swing2LimitDegrees=45
  swing2Motion=Free
  twistLimitDegrees=45
  twistMotion=Free
 }
 {
  bone1Name=spine_05
  bone2Name=pelvis
  swing1LimitDegrees=45
  swing1Motion=Free
  swing2LimitDegrees=45
  swing2Motion=Free
  twistLimitDegrees=45
  twistMotion=Free
 }
]
