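# Physics asset for the dynamic back of the poncho on the adult male medium Headband character body.
# Three capsule chains hang from tiny master bodies on spine_05 and both clavicles, with angular limits shaping their swing and twist.
# Free-angle constraints tie both side-chain tips to the center chain's end, linking the panels.
# Torso, upper-arm and leg capsules supply body shapes, with larger capsules on the right leg than the left.
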PhysicsAsset /BRCosmetics/Characters/Player/Male/Medium/Bodies/M_MED_Headband_Adult/Meshes/M_MED_Headband_Adult_Dyn_poncho_back
bodies {
 calf_l [{shapeType=Capsule, length=27.125, radius=7.695, center=<-20.9669 -2.842 0>, rotation=<rot 90 0 6.4995>}]
 calf_r [{shapeType=Capsule, length=30.375, radius=10.2575, center=<20.9669 2.842 0>, rotation=<rot 90 0 -173.5005>}]
 clavicle_l [{shapeType=Capsule, length=8, radius=7.1325, center=<8.504 0.2701 -0.5883>, rotation=<rot 90 0 0>}]
 clavicle_r [{shapeType=Capsule, length=8.8437, radius=6.2888, center=<-8.504 0.0004 0>, rotation=<rot 90 0 180>}]
 dyn_poncho_bk_02 [{
   shapeType=Capsule
   length=20.3537
   radius=11.2771
   shapeName=dyn_poncho_elbow_l_03_capsule
   center=<10.5403 0.0001 10.5>
   rotation=<rot 90 0 0>
  }]
 dyn_poncho_bk_03 [{
   shapeType=Capsule
   length=20.3537
   radius=11.2771
   shapeName=dyn_poncho_elbow_l_03_capsule
   center=<10.5403 0.0001 10.5>
   rotation=<rot 90 0 0>
  }]
 dyn_poncho_bk_MSTR [{shapeType=Capsule, length=1, radius=1, shapeName=dyn_poncho_elbow_l_03_capsule}]
 dyn_poncho_bk_l_02 [{
   shapeType=Capsule
   length=20.3537
   radius=11.2771
   shapeName=dyn_poncho_elbow_l_03_capsule
   center=<10.5403 0.0001 10.5>
   rotation=<rot 90 0 0>
  }]
 dyn_poncho_bk_l_03 [{shapeType=Capsule, length=26.4162, radius=15, shapeName=dyn_poncho_elbow_l_03_capsule, center=<10.5403 0 15>, rotation=<rot 90 0 0>}]
 dyn_poncho_bk_l_04 [{shapeType=Capsule, length=26.4162, radius=16, shapeName=dyn_poncho_elbow_l_03_capsule, center=<10.5403 0 15>, rotation=<rot 90 0 0>}]
 dyn_poncho_bk_l_MSTR [{shapeType=Capsule, length=1, radius=1, shapeName=dyn_poncho_elbow_l_03_capsule, rotation=<rot 90 0 0>}]
 dyn_poncho_bk_r_02 [{
   shapeType=Capsule
   length=20.3537
   radius=11.2771
   shapeName=dyn_poncho_elbow_l_03_capsule
   center=<10.5403 0.0001 10.5>
   rotation=<rot 90 0 0>
  }]
 dyn_poncho_bk_r_03 [{shapeType=Capsule, length=26.4162, radius=15, shapeName=dyn_poncho_elbow_l_03_capsule, center=<10.5403 0 15>, rotation=<rot 90 0 0>}]
 dyn_poncho_bk_r_04 [{shapeType=Capsule, length=26.4162, radius=16, shapeName=dyn_poncho_elbow_l_03_capsule, center=<10.5403 0 15>, rotation=<rot 90 0 0>}]
 dyn_poncho_bk_r_MSTR [{shapeType=Capsule, length=1, radius=1, shapeName=dyn_poncho_elbow_l_03_capsule, rotation=<rot 90 0 0>}]
 dyn_poncho_fr_l_MSTR [{shapeType=Capsule, length=1, radius=1, shapeName=dyn_poncho_elbow_l_03_capsule}]
 dyn_poncho_fr_r_MSTR [{shapeType=Capsule, length=1, radius=1, shapeName=dyn_poncho_elbow_l_03_capsule}]
 pelvis [
  {shapeType=Capsule, length=16.9687, radius=14.7263, center=<-2.6215 -0.3628 0>}
  {shapeType=Capsule, length=11.4062, radius=17.82, shapeName=pelvis_capsule, center=<-11.5935 0.2438 0>}
 ]
 spine_01 [{shapeType=Capsule, length=12.7812, radius=16.945, center=<-2.6215 -0.045 -0.0001>}]
 spine_02 [
  {shapeType=Capsule, length=6.4687, radius=17.6013, center=<-2.6215 0.4993 0.0001>}
  {shapeType=Capsule, length=29.875, radius=11.1013, center=<2.5024 0.4992 3.9825>, rotation=<rot 90 0.0049 0.0049>}
  {shapeType=Capsule, length=29.875, radius=11.1013, center=<2.4704 0.4991 -3.9825>, rotation=<rot 90 -0.0396 -0.0396>}
 ]
 spine_03 [{shapeType=Capsule, length=5.6875, radius=15.6638, center=<-1.1937 -0.9562 0.0002>}]
 spine_04 [{shapeType=Capsule, length=9.4687, radius=13.4138, center=<-0.0327 -0.5317 0.0001>}]
 spine_05 [
  {shapeType=Capsule, length=13.1562, radius=12.2888, center=<-2.1747 -1.0553 0>}
  {shapeType=Capsule, length=17.6562, radius=10.8125, center=<-0.542 -1.0479 0>}
 ]
 thigh_l [{shapeType=Capsule, length=21.9062, radius=8.2263, center=<-22.0309 -0.0003 0>, rotation=<rot 90 0 0>}]
 thigh_r [{shapeType=Capsule, length=34, radius=10.9763, center=<22.0309 0.0003 0>, rotation=<rot 90 0 180>}]
 upperarm_l [{shapeType=Capsule, length=23.8437, radius=6.2888, center=<11.345 -0.0011 0>, rotation=<rot 90 0 0>}]
 upperarm_r [{shapeType=Capsule, length=23.8437, radius=6.2888, center=<-11.345 0.0011 0>, rotation=<rot 90 0 180>}]
}
constraints [
 {
  bone1Name=spine_03
  bone2Name=spine_02
  swing1LimitDegrees=45
  swing1Motion=Locked
  swing2LimitDegrees=45
  swing2Motion=Locked
  twistLimitDegrees=45
  twistMotion=Locked
 }
 {
  bone1Name=spine_05
  bone2Name=spine_04
  swing1LimitDegrees=45
  swing1Motion=Locked
  swing2LimitDegrees=45
  swing2Motion=Locked
  twistLimitDegrees=45
  twistMotion=Locked
 }
 {
  bone1Name=spine_02
  bone2Name=spine_01
  swing1LimitDegrees=45
  swing1Motion=Locked
  swing2LimitDegrees=45
  swing2Motion=Locked
  twistLimitDegrees=45
  twistMotion=Locked
 }
 {
  bone1Name=spine_01
  bone2Name=pelvis
  swing1LimitDegrees=45
  swing1Motion=Locked
  swing2LimitDegrees=45
  swing2Motion=Locked
  twistLimitDegrees=45
  twistMotion=Locked
 }
 {
  bone1Name=spine_04
  bone2Name=spine_03
  swing1LimitDegrees=45
  swing1Motion=Locked
  swing2LimitDegrees=45
  swing2Motion=Locked
  twistLimitDegrees=45
  twistMotion=Locked
 }
 {
  bone1Name=dyn_poncho_bk_l_MSTR
  bone2Name=clavicle_l
  swing1LimitDegrees=45
  swing1Motion=Locked
  swing2LimitDegrees=45
  swing2Motion=Locked
  twistLimitDegrees=5
  twistMotion=Locked
 }
 {
  bone1Name=dyn_poncho_bk_l_02
  bone2Name=dyn_poncho_bk_l_MSTR
  swing1LimitDegrees=5
  swing1Motion=Locked
  swing2LimitDegrees=35
  twistLimitDegrees=5
  twistMotion=Locked
 }
 {bone1Name=dyn_poncho_bk_l_03, bone2Name=dyn_poncho_bk_l_02, swing1LimitDegrees=5, swing2LimitDegrees=35, twistLimitDegrees=5}
 {
  bone1Name=dyn_poncho_bk_l_04
  bone2Name=dyn_poncho_bk_l_03
  swing1LimitDegrees=30
  swing2LimitDegrees=35
  twistLimitDegrees=5
  twistMotion=Locked
 }
 {
  bone1Name=dyn_poncho_bk_r_MSTR
  bone2Name=clavicle_r
  swing1LimitDegrees=45
  swing1Motion=Locked
  swing2LimitDegrees=45
  swing2Motion=Locked
  twistLimitDegrees=5
  twistMotion=Locked
 }
 {
  bone1Name=dyn_poncho_bk_r_02
  bone2Name=dyn_poncho_bk_r_MSTR
  swing1LimitDegrees=5
  swing1Motion=Locked
  swing2LimitDegrees=35
  twistLimitDegrees=5
  twistMotion=Locked
 }
 {bone1Name=dyn_poncho_bk_r_03, bone2Name=dyn_poncho_bk_r_02, swing1LimitDegrees=5, swing2LimitDegrees=35, twistLimitDegrees=5}
 {
  bone1Name=dyn_poncho_bk_r_04
  bone2Name=dyn_poncho_bk_r_03
  swing1LimitDegrees=30
  swing2LimitDegrees=35
  twistLimitDegrees=5
  twistMotion=Locked
 }
 {bone1Name=dyn_poncho_bk_MSTR, bone2Name=spine_05, swing1LimitDegrees=45, swing2LimitDegrees=45, twistLimitDegrees=5}
 {bone1Name=dyn_poncho_bk_02, bone2Name=dyn_poncho_bk_MSTR, swing1LimitDegrees=1, swing2LimitDegrees=35, twistLimitDegrees=5}
 {
  bone1Name=dyn_poncho_bk_03
  bone2Name=dyn_poncho_bk_02
  swing1LimitDegrees=10
  swing2LimitDegrees=60
  twistLimitDegrees=5
  twistMotion=Locked
 }
 {
  bone1Name=clavicle_l
  bone2Name=spine_05
  swing1LimitDegrees=45
  swing1Motion=Free
  swing2LimitDegrees=45
  swing2Motion=Free
  twistLimitDegrees=45
  twistMotion=Free
 }
 {
  bone1Name=upperarm_l
  bone2Name=clavicle_l
  swing1LimitDegrees=45
  swing1Motion=Free
  swing2LimitDegrees=45
  swing2Motion=Free
  twistLimitDegrees=45
  twistMotion=Free
 }
 {
  bone1Name=thigh_l
  bone2Name=pelvis
  swing1LimitDegrees=45
  swing1Motion=Free
  swing2LimitDegrees=45
  swing2Motion=Free
  twistLimitDegrees=45
  twistMotion=Free
 }
 {
  bone1Name=calf_l
  bone2Name=thigh_l
  swing1LimitDegrees=45
  swing1Motion=Free
  swing2LimitDegrees=45
  swing2Motion=Free
  twistLimitDegrees=45
  twistMotion=Free
 }
 {
  bone1Name=clavicle_r
  bone2Name=spine_05
  swing1LimitDegrees=45
  swing1Motion=Free
  swing2LimitDegrees=45
  swing2Motion=Free
  twistLimitDegrees=45
  twistMotion=Free
 }
 {
  bone1Name=upperarm_r
  bone2Name=clavicle_r
  swing1LimitDegrees=45
  swing1Motion=Free
  swing2LimitDegrees=45
  swing2Motion=Free
  twistLimitDegrees=45
  twistMotion=Free
 }
 {
  bone1Name=thigh_r
  bone2Name=pelvis
  swing1LimitDegrees=45
  swing1Motion=Free
  swing2LimitDegrees=45
  swing2Motion=Free
  twistLimitDegrees=45
  twistMotion=Free
 }
 {
  bone1Name=calf_r
  bone2Name=thigh_r
  swing1LimitDegrees=45
  swing1Motion=Free
  swing2LimitDegrees=45
  swing2Motion=Free
  twistLimitDegrees=45
  twistMotion=Free
 }
 {
  bone1Name=dyn_poncho_fr_l_MSTR
  bone2Name=clavicle_l
  swing1LimitDegrees=45
  swing1Motion=Free
  swing2LimitDegrees=45
  swing2Motion=Free
  twistLimitDegrees=45
  twistMotion=Free
 }
 {
  bone1Name=dyn_poncho_fr_r_MSTR
  bone2Name=clavicle_r
  swing1LimitDegrees=45
  swing1Motion=Free
  swing2LimitDegrees=45
  swing2Motion=Free
  twistLimitDegrees=45
  twistMotion=Free
 }
 {
  bone1Name=dyn_poncho_bk_l_04
  bone2Name=dyn_poncho_bk_03
  swing1LimitDegrees=45
  swing1Motion=Free
  swing2LimitDegrees=45
  swing2Motion=Free
  twistLimitDegrees=45
  twistMotion=Free
 }
 {
  bone1Name=dyn_poncho_bk_r_04
  bone2Name=dyn_poncho_bk_03
  swing1LimitDegrees=45
  swing1Motion=Free
  swing2LimitDegrees=45
  swing2Motion=Free
  twistLimitDegrees=45
  twistMotion=Free
 }
]
